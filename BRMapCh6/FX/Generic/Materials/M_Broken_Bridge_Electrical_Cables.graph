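# Default Lit material for broken bridge electrical cables, using PowerPlant broken-wire diffuse and normal textures.
# Diffuse color is multiplied by vertex color, with a Desaturation parameter applied to that vertex color.
# An inverted V gradient on the second UV channel is squared and raised to a high power, forming a narrow end mask.
# That mask confines an emissive color driven by a time sine, with vertex red setting flicker speed.
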Material /BRMapCh6/FX/Generic/Materials/M_Broken_Bridge_Electrical_Cables
graph {
 properties {bAutomaticallySetUsageInEditor=false, usedShadingModels=Default Lit}
 nodes [
  n1 = Multiply(A=VertexColor().R, B=Time())
  n2 = TextureCoordinate() {coordinateIndex=1}
  n3 = Multiply(A=OneMinus(None=MaterialFunctionCall(UV Channel=@n2) {materialFunction=&LinearGradient}.VGradient))
  n4 = Multiply(A=@n3, B=@n3)
  n5 = Power(Base=@n4) {constExponent=150}
  LinearInterpolate(B=Multiply(A=Sine(None=@n1), B=VectorParameter() {parameterName=EmmisiveColor, defaultValue=<rgba 25 2 1 1>}.RGB), Alpha=@n5) {constB=0}
  Multiply(
   A=TextureSample() {texture=/Game/Environments/Apollo/Sets/PowerPlant/Props/Textures/T_PP_CableWires_Broken_D}.RGB
   B=Desaturation(None=VertexColor(), Fraction=ScalarParameter() {parameterName=Desaturation})
  )
  OneMinus(None=@n5)
  Saturate(None=@n4)
  TextureSample() {samplerType=SAMPLERTYPE_Normal, texture=/Game/Environments/Apollo/Sets/PowerPlant/Props/Textures/T_PP_CableWires_Broken_N}
 ]
}
